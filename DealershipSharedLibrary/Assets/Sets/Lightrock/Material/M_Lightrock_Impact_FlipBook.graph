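# Translucent, unlit, two-sided Niagara particle material for Lightrock impact effects, blending a dissolving base-shape look with a spark "Flipbook" look, apparently via a dynamic parameter.
# Two rotated, panning noise samples, split across left and right UV halves and offset per particle, form a mask that breaks up both looks.
# The base uses two-color tinting or particle color through PaintedVariant; sparks use particle color and dynamic dissolve.
# Depth fade softens opacity.
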Material /DealershipSharedLibrary/Assets/Sets/Lightrock/Material/M_Lightrock_Impact_FlipBook
graph {
 properties {
  bAutomaticallySetUsageInEditor=false
  bUsedWithMeshParticles=true
  bUsedWithNiagaraMeshParticles=true
  bUsedWithNiagaraRibbons=true
  bUsedWithNiagaraSprites=true
  blendMode=BLEND_Translucent
  shadingModel=MSM_Unlit
  twoSided=true
  usedShadingModels=Unlit
 }
 nodes [
  DepthFade(FadeDistance=ScalarParameter() {defaultValue=6000, parameterName=Depth Fade})
  n1 = NamedRerouteUsage()
  LinearInterpolate(A=NamedRerouteUsage(), B=NamedRerouteUsage(), Alpha=@n1)
  MaterialFunctionCall(Emissive=LinearInterpolate(A=NamedRerouteUsage(), B=NamedRerouteUsage(), Alpha=@n1)) {materialFunction=&MF_FX_EmissiveAdjustments}
  n2 = ScalarParameter() {defaultValue=1, parameterName=MaskPower}
  n3 = ScalarParameter() {defaultValue=-0.18, parameterName=MaskLift_Speed}
  n4 = Cosine(None=@n3)
  n5 = ScalarParameter() {defaultValue=2, parameterName=MaskLift_U}
  n6 = ScalarParameter() {defaultValue=1, parameterName=MaskLift_V}
  n7 = Constant() {r=0.5}
  n8 = Subtract(A=Multiply(A=TextureCoordinate() {uTiling=2}, B=AppendVector(A=@n5, B=@n6)), B=@n7)
  n9 = ComponentMask(None=@n8) {r=true}
  n10 = Multiply(A=@n4, B=@n9)
  n11 = Sine(None=@n3)
  n12 = ComponentMask(None=@n8) {g=true}
  n13 = Multiply(A=@n11, B=@n12)
  n14 = Multiply(A=@n4, B=@n12)
  n15 = Multiply(A=@n11, B=@n9)
  n16 = Add(A=AppendVector(A=Add(A=@n10, B=@n13), B=Subtract(A=@n14, B=@n15)), B=@n7)
  n17 = NamedRerouteUsage()
  n18 = AppendVector(A=Multiply(A=NamedRerouteUsage()) {constB=-1}, B=NamedRerouteUsage())
  n19 = Time()
  n20 = ParticleRandom()
  n21 = Add(A=Add(A=Multiply(A=@n16, B=@n17), B=Multiply(A=@n18, B=@n19)), B=@n20)
  n22 = ComponentMask(None=TextureCoordinate()) {r=true}
  n23 = ConstantBiasScale(None=@n22) {bias=0.5, scale=1}
  n24 = Constant() {r=0.2}
  n25 = Cosine(None=@n24)
  n26 = Constant() {r=0.5}
  n27 = Subtract(A=TextureCoordinate() {uTiling=2}, B=@n26)
  n28 = ComponentMask(None=@n27) {r=true}
  n29 = Multiply(A=@n25, B=@n28)
  n30 = Sine(None=@n24)
  n31 = ComponentMask(None=@n27) {g=true}
  n32 = Multiply(A=@n30, B=@n31)
  n33 = Multiply(A=@n25, B=@n31)
  n34 = Multiply(A=@n30, B=@n28)
  n35 = Add(A=AppendVector(A=Add(A=@n29, B=@n32), B=Subtract(A=@n33, B=@n34)), B=@n26)
  n36 = NamedRerouteUsage()
  n37 = AppendVector(A=NamedRerouteUsage(), B=NamedRerouteUsage())
  n38 = Time()
  n39 = Add(A=Add(A=Multiply(A=@n35, B=@n36), B=Multiply(A=@n37, B=@n38)), B=@n20)
  n40 = ConstantBiasScale(None=OneMinus(None=@n22)) {bias=0.5, scale=1}
  n41 = Add(
   A=Multiply(
    A=TextureSampleParameter2D(UVs=@n21) {parameterName=MaskLiftTextures, texture=/DealershipSharedLibrary/FX/Textures/Gradient_Masks/T_VCVFX_Boost_Anisotropic_A}.RGB
    B=Floor(None=@n23)
   )
   B=Multiply(
    A=TextureSampleParameter2D(UVs=@n39) {parameterName=MaskRightTextures, texture=&T_VCVFX_Boost_Anisotropic_A}.RGB
    B=Floor(None=@n40)
   )
  )
  NamedRerouteDeclaration(None=Multiply(A=Power(Base=@n41, Exp=@n2) {constExponent=1.2}, B=ScalarParameter() {defaultValue=1, parameterName=MaskIntensity})) {name=Mask, nodeColor=<rgba 1 0.7266 0 1>}
  n42 = ParticleColor()
  n43 = DynamicParameter() {parameterIndex=1, defaultValue=<rgba 0 0.2 1 1>, paramNames [FlipBookDissolve, FlipBookFeather, Param3, Param4]}
  n44 = SmoothStep(
   Min=@n43.FlipBookDissolve
   Max=Add(A=@n43.FlipBookDissolve, B=@n43.FlipBookDissolve)
   Value=TextureSample() {texture=/DealershipSharedLibrary/Assets/Sets/Lightrock/Textures/T_Lightrock_Spark_RGB}.R
  )
  n45 = Multiply(
   A=Power(Base=@n41, Exp=ScalarParameter() {defaultValue=4, parameterName=Power}) {constExponent=1.2}
   B=ScalarParameter() {defaultValue=2, parameterName=Int}
  )
  n46 = Add(A=@n44, B=Multiply(A=@n44, B=Saturate(None=@n45)))
  NamedRerouteDeclaration(None=Multiply(A=@n42.RGB, B=@n46)) {name=Flipbook, nodeColor=<rgba 0 0.9766 1 1>}
  NamedRerouteDeclaration(None=DepthFade(Opacity=Multiply(A=@n42.A, B=@n46), FadeDistance=ScalarParameter() {defaultValue=6000, parameterName=Depth Fade})) {name=FlipbookOpacity, nodeColor=<rgba 1 0.7031 0 1>}
  n47 = DynamicParameter() {defaultValue=<rgba 1 0.2 0 1>, paramNames [NoiseUSpeed, NoiseVSpeed, "Mask&&Filpbook", NoiseTilling]}
  NamedRerouteDeclaration(None=@n47."Mask&&Filpbook") {name="Mask&&Filpbook", nodeColor=<rgba 0 1 0.25 1>}
  NamedRerouteDeclaration(None=Multiply(A=@n47.NoiseVSpeed) {constB=-1}) {name=NoiseVSpeed, nodeColor=<rgba 1 0.3281 0 1>}
  n48 = ScalarParameter() {defaultValue=0.4, parameterName=CenterDissolve}
  n49 = TextureSampleParameter2D() {parameterName=Outer Textures, texture=/DealershipSharedLibrary/Assets/Sets/Lightrock/Textures/T_Lightrock_Base_RGB}
  n50 = NamedRerouteUsage()
  n51 = SmoothStep(Min=@n48, Max=Add(A=@n48, B=ScalarParameter() {defaultValue=0.2, parameterName=CenterFeather}), Value=Multiply(A=@n49.R, B=@n50))
  n52 = ScalarParameter() {defaultValue=0.2, parameterName=OuterFeather}
  n53 = ScalarParameter() {defaultValue=0.5, parameterName=OuterDissolve}
  n54 = ScalarParameter() {defaultValue=1.5, parameterName=CenterScaleUV_U}
  n55 = Constant() {r=1}
  n56 = TextureSampleParameter2D(UVs=MaterialFunctionCall(Texture Scale=AppendVector(A=@n54, B=@n55)) {materialFunction=&ScaleUVsByCenter}.UVs) {parameterName=Center Textures, texture=&T_Lightrock_Base_RGB}
  n57 = Max(A=SmoothStep(Min=@n53, Max=Add(A=@n53, B=@n52), Value=Multiply(A=@n50, B=@n56.R)), B=@n56.G)
  n58 = Add(
   A=Multiply(A=VectorParameter() {parameterName=Center Color, defaultValue=<rgba 0.2656 0.3831 1 1>}.RGB, B=Max(A=@n49.G, B=@n51))
   B=Multiply(A=@n57, B=VectorParameter() {parameterName=Outer Color, defaultValue=<rgba 0.0006 0.1012 1 1>}.RGB)
  )
  NamedRerouteDeclaration(None=If(A=ScalarParameter() {parameterName=PaintedVariant}, "A > B"=ParticleColor().RGB, "A == B"=@n58, "A < B"=@n58)) {name=EmissiveColor, nodeColor=<rgba 0 0.7422 1 1>}
  NamedRerouteDeclaration(None=@n47.NoiseTilling) {name=NoiseTilling, nodeColor=<rgba 1 0 0.8672 1>}
  n59 = ParticleColor()
  NamedRerouteDeclaration(None=Multiply(A=Multiply(A=@n57, B=@n59.A), B=Subtract(A=@n56.A) {constB=0})) {name=Opacity, nodeColor=<rgba 0.6641 1 0 1>}
  NamedRerouteDeclaration(None=@n47.NoiseUSpeed) {name=NoiseUSpeed, nodeColor=<rgba 0 0.3906 1 1>}
 ]
}
